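# Niagara vehicle boost cosmetic that trails a ribbon and sprays spinning flower sprites from the car's filtered boost sockets.
# Two matching sprite emitters partition particles across sockets, then move them with cone velocity, curved gravity, vortex and inherited motion.
# A random type picks the SubUV frame for three renderers, while a random hue indexes paired inner/outer color arrays.
# Painted variants can override sprite colors, and a frontend-preview flag changes color scale and size curves.
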NiagaraSystem /VehicleCosmetics/Boosts/Boost_Flowers/FX/NS_Boost_Flowers
dependencies {
 usedDataInterfaces [
  &NiagaraDataInterfaceCurve
  &NiagaraDataInterfaceArrayColor
  &NiagaraDataInterfaceVector2DCurve
  &NiagaraDataInterfaceArrayInt32
  &NiagaraDataInterfaceSpriteRendererInfo
  &NiagaraDataInterfaceSkeletalMesh
 ]
 usedDynamicInputs [&RandomRangeVector]
 usedModules [
  &self:SystemSpawnScript.SystemScriptSource.SystemScriptGraph.NiagaraNodeAssignment_0.SetVariables_#1
  &SystemState
 ]
}
emitters {
 NE_VC_Boost_Ribbon_Parent {
  emitterUpdateScript [
   &EmitterState()
   &VC_FilteredSocketCount()
   &VC_RateQualityHandler(Epic Spawn Rate=30, High Spawn Rate=30, Low Spawn Rate=30, Use Custom Rate?=true)
   &VC_SpawnRibbon_v1(Number of Ribbons=$Emitter.PortCount, SpawnRate=$Emitter.Spawn Rate)
  ]
  particleSpawnScript [
   &VC_PaintedVariantsOverride() {enabled=false}
   &VC_InitializeRibbon_v1(
    Custom Offset=true
    Enable Skeletal Position=true
    Lifetime=1.5
    Local Vector Offset=&AddVectorToPosition(Position=<0 0 0>, Vector=$User.DTCarOffset)
    Offset Amount=1
    Ribbon Width=90
   )
   &AddVelocity(Velocity=<-1500 0 0>) {module=AddVelocity001}
  ]
  particleUpdateScript [
   &ParticleState()
   &InheritSourceMovement(
    Applied Position Scale=&VectorFromFloat(Value=&FloatFromCurve(FloatCurve=DataInterface() {propertyValues {Curve {keys [
          {interpMode=RCIM_Cubic, tangentMode=RCTM_Break, time=0, value=1}
          {arriveTangent=-2.9025, interpMode=RCIM_Cubic, tangentMode=RCTM_Break, time=1, value=0}
         ]}}}))
    ApplyForce=false
    ApplyPosition=true
    ApplyPositionFalloff=false
   )
   &SolveForcesAndVelocity(Clamp Velocity=false, Limit Acceleration=false)
   &VC_UpdateRibbon_v1(Enable Skeletal Position=true)
   &DynamicMaterialParameters(
    Module.Index0Float0Write=true
    Module.Index0Float1Write=true
    Module.Index0Float2Write=true
    Module.Index0Float3Write=true
    Module.Index1Float0Write=true
    Module.Index1Float1Write=true
    Module.Index1Float2Write=true
    Module.Index1Float3Write=true
    Write Parameter Index 1=::NewEnumerator1
   )
   &ScaleRibbonWidth(Ribbon Width Scale=&FloatFromCurve(FloatCurve=DataInterface() {propertyValues {Curve {keys [
         <at 0 0.75 RCIM_Cubic>
         {arriveTangent=-0.0833, interpMode=RCIM_Cubic, leaveTangent=-0.0833, time=0.15, value=0.65}
         {arriveTangent=-0.4782, interpMode=RCIM_Cubic, leaveTangent=-0.4782, tangentMode=RCTM_User, time=0.6, value=0.7}
         {arriveTangent=-5.6669, interpMode=RCIM_Cubic, leaveTangent=-5.6669, tangentMode=RCTM_User, time=0.95, value=0}
        ]}}}))
   &ScaleColor(
    Scale Alpha=&FloatFromCurve(FloatCurve=DataInterface() {propertyValues {Curve {keys [<at 0 1>,<at 1 0>]}}})
    ScaleA=true
    ScaleRGB=true
   )
  ]
  rendererClasses [&NiagaraRibbonRendererProperties]
  renderers [{rendererClass=&NiagaraRibbonRendererProperties}]
 }
 Sprites {
  emitterUpdateScript [
   &EmitterState()
   &VC_FilteredSocketCount()
   &SpawnRate(SpawnRate=90, Use Spawn Probability=false)
  ]
  particleSpawnScript [
   /SetVariables(
    Particles.FlowerHue:NiagaraInt32=&RandomRangeInt(Minimum=0, Maximum=9, Randomness Mode=::NewEnumerator1)
    Particles.FlowerType:NiagaraFloat=&MakeFloatFromInt(Int=&RandomRangeInt(Minimum=0, Maximum=3))
   ) {isSetParameters=true, module=SetVariables_#2}
   /SetVariables(
    Particles.BackFillColor:LinearColor=<rgba 0 0 0 2>
    Particles.InnerColor:LinearColor=&MultiplyLinearColor(
     Linear Color=&SelectLinearColorFromArray(
      Array Sampling Mode=::NewEnumerator1
      Direct Array Index=$Particles.FlowerHue
      Color Selection Array=DataInterface() {propertyValues {ColorData [
         <rgba 1 0.4697 0.0024 1>
         <rgba 0.8333 0 0.4982 1>
         <rgba 0 0.9002 1 1>
         <rgba 0.2007 0.0659 1 1>
         <rgba 0.8177 0.2491 0 1>
         <rgba 0.0035 1 0.989 1>
         <rgba 0 1 0.3343 1>
         <rgba 1 0 0.0712 1>
         <rgba 1 0.0083 0.6752 1>
         <rgba 0 0.0252 1 1>
        ]}}
      Random Seed=?
     )
     Scale Factor=/Niagara/DynamicInputs/TypeConversions/MakeCustomVector4FromBool(Bool:NiagaraBool=$User.bIsFrontendPreview, True Vector4:Vector4f=<0.6 0.6 0.6 2.5>, False Vector4:Vector4f=<1.1 1.1 1.1 2>)
    )
    Particles.OuterColor:LinearColor=&MultiplyLinearColor(
     Linear Color=&SelectLinearColorFromArray(
      Array Sampling Mode=::NewEnumerator1
      Direct Array Index=$Particles.FlowerHue
      Color Selection Array=DataInterface() {propertyValues {ColorData [
         <rgba 0.0203 0 0.3698 1>
         <rgba 0 1 0.2287 1>
         <rgba 1 0.0001 0.0227 1>
         <rgba 0.9182 1 0 1>
         <rgba 0.0754 0.6907 0.8 1>
         <rgba 1 0.3226 0 1>
         <rgba 0.6121 0.007 1 1>
         <rgba 0 0.3588 1 1>
         <rgba 0 1 0.1269 1>
         <rgba 1 0 0.007 1>
        ]}}
      Random Seed=?
     )
     Scale Factor=&MakeCustomVector4FromBool(Bool:NiagaraBool=$User.bIsFrontendPreview, True Vector4:Vector4f=<0.6 0.6 0.6 2.5>, False Vector4:Vector4f=<1.1 1.1 1.1 2>)
    )
   ) {isSetParameters=true, module=SetVariables_#3}
   &VC_PaintedVariantsOverride(Black=<rgba 0.2517 0.2517 0.2517 1>, BurntSienna=<rgba 0.3177 0.0981 0.0392 1>)
   &InitializeParticle(
    Color Mode=::NewEnumerator0
    Lifetime Max=0.9
    Lifetime Min=0.25
    Lifetime Mode=::NewEnumerator1
    Sprite Rotation Mode=::NewEnumerator2
    Sprite Size Mode=::NewEnumerator1
    Uniform Sprite Size Max=42
    Uniform Sprite Size Min=17.5
    UsePositionOffset=false
    Write Material Random=true
   )
   &PartitionParticles(Number Of Partitions=$Emitter.PortCount)
   &SkeletalMeshLocation(
    Bone / Socket Index=$Particles.PartitionParticles.Partition
    Bone / Socket Sampling Mode=::NewEnumerator5
    MaskSpawnGroup=false
    Mesh Sampling Type=::NewEnumerator6
    Sampled Position Offset=&AddVectorToPosition(Position=<-15 0 0>, Vector=$User.DTCarOffset)
    Skeletal Mesh=$System.SkeletalMesh
   )
   &AddVelocity(
    Cone Angle=20
    Cone Axis=<0 0 1>
    Module.UseConeFalloff=false
    Module.UseVelDistribution=false
    Override Seed=false
    Rotation Mode=::NewEnumerator2
    Velocity Mode=::NewEnumerator2
    Velocity Speed=2000
    Yaw / Pitch / Roll:Vector3f=<0 82.5 0>
   )
   &ShapeLocation(MaskSpawnGroup=false, Override Seed=false, Sphere Radius=30, UseOwnerScale=true)
  ]
  particleUpdateScript [
   &ParticleState()
   &ScaleSpriteSize(
    Scale Sprite Size Mode=::NewEnumerator0
    Uniform Scale Factor=&MakeCustomFloatFromBool(
     Bool=$User.bIsFrontendPreview
     True Float=&FloatFromCurve(FloatCurve=DataInterface() {propertyValues {Curve {
         defaultValue=1
         keys [
          <at 0 0.5 RCIM_Cubic>
          {arriveTangent=-0.0087, interpMode=RCIM_Cubic, leaveTangent=-0.0087, tangentMode=RCTM_User, time=0.075, value=1}
          {arriveTangent=-0.1297, interpMode=RCIM_Cubic, leaveTangent=-0.1297, tangentMode=RCTM_User, time=0.6, value=1}
          {arriveTangent=-4.2097, interpMode=RCIM_Cubic, leaveTangent=-4.2097, tangentMode=RCTM_User, time=1, value=0}
         ]
        }}})
     False Float=&FloatFromCurve(FloatCurve=DataInterface() {propertyValues {Curve {
         defaultValue=1
         keys [
          <at 0 1 RCIM_Cubic>
          {arriveTangent=-0.123, interpMode=RCIM_Cubic, leaveTangent=-0.123, tangentMode=RCTM_User, time=0.55, value=1.25}
          {arriveTangent=-3.7842, interpMode=RCIM_Cubic, leaveTangent=-3.7842, tangentMode=RCTM_User, time=1, value=0}
         ]
        }}})
    )
   )
   &SpriteRotationRate(Rotation Rate=&RandomRangeFloat(Minimum=0.5, Maximum=360))
   &GravityForce(Gravity=&MakeVector(
     X=0
     Y=0
     Z=&FloatFromCurve(FloatCurve=DataInterface() {propertyValues {Curve {keys [
          {arriveTangent=4327.0386, interpMode=RCIM_Cubic, leaveTangent=4327.0386, tangentMode=RCTM_User, time=0, value=500}
          {arriveTangent=-162.7294, interpMode=RCIM_Cubic, leaveTangent=-162.7294, tangentMode=RCTM_User, time=0.025, value=700}
          <at 0.1 -1000 RCIM_Cubic>
         ]}}})
    )) {module=GravityForce004}
   &VortexForce(
    UseOriginOffset=false
    Vortex Axis=<-1 0 0>
    Vortex Force Amount=&Multiply_FloatByInt(
     Float=&RandomRangeFloat(Minimum=100, Maximum=500)
     Integer=&SelectIntFromArray(
      Array Sampling Mode=::NewEnumerator1
      Direct Array Index=?
      Int Selection Array=DataInterface() {propertyValues {IntData [-1,1]}}
      Random Seed=0
     )
    )
   )
   &InheritSourceMovement(
    Applied Position Scale=&Multiply_VectorByFloat(
     Vector=<1 1 1>
     Float=&FloatFromCurve(FloatCurve=DataInterface() {propertyValues {Curve {keys [{interpMode=RCIM_Cubic, tangentMode=RCTM_User, time=0.3, value=1}, {interpMode=RCIM_Cubic, tangentMode=RCTM_User, time=1, value=0}]}}})
    )
    ApplyForce=false
    ApplyPosition=true
    ApplyPositionFalloff=false
   )
   &SolveForcesAndVelocity(Clamp Velocity=false, Limit Acceleration=false)
   &SubUVAnimation(
    Direct Frame Index:NiagaraFloat=$Particles.FlowerType
    "Sprite Renderer "=DataInterface() {propertyValues {SpriteRenderer=&self:Sprites_0.NiagaraSpriteRendererProperties_3}}
    SubUV Animation Mode:ENiagara_SubUVLookupModeV2=::NewEnumerator5
   )
   &SubUVAnimation(
    Direct Frame Index:NiagaraFloat=$Particles.FlowerType
    "Sprite Renderer "=DataInterface() {propertyValues {SpriteRenderer=&self:Sprites_0.NiagaraSpriteRendererProperties_0}}
    SubUV Animation Mode:ENiagara_SubUVLookupModeV2=::NewEnumerator5
   ) {module=SubUVAnimation001}
   &SubUVAnimation(
    Direct Frame Index:NiagaraFloat=$Particles.FlowerType
    "Sprite Renderer "=DataInterface() {propertyValues {SpriteRenderer=&self:Sprites_0.NiagaraSpriteRendererProperties_1}}
    SubUV Animation Mode:ENiagara_SubUVLookupModeV2=::NewEnumerator5
   ) {module=SubUVAnimation002}
   &JitterPosition(Jitter Amount=2)
   &ScaleColor(Scale RGB=&VectorFromFloat(Value=1), ScaleA=true, ScaleRGB=true)
  ]
  rendererClasses [&NiagaraSpriteRendererProperties]
  renderers [
   {rendererClass=&NiagaraSpriteRendererProperties}
   {rendererClass=&NiagaraSpriteRendererProperties, rendererIndex=1}
   {rendererClass=&NiagaraSpriteRendererProperties, rendererIndex=2}
  ]
 }
 Sprites001 {
  emitterUpdateScript [
   &EmitterState()
   &VC_FilteredSocketCount()
   &SpawnRate(SpawnRate=90, Use Spawn Probability=false)
  ]
  particleSpawnScript [
   /SetVariables(
    Particles.FlowerHue:NiagaraInt32=&RandomRangeInt(Minimum=0, Maximum=9, Randomness Mode=::NewEnumerator1)
    Particles.FlowerType:NiagaraFloat=&MakeFloatFromInt(Int=&RandomRangeInt(Minimum=0, Maximum=3))
   ) {isSetParameters=true, module=SetVariables_#2}
   /SetVariables(
    Particles.BackFillColor:LinearColor=<rgba 0 0 0 2>
    Particles.InnerColor:LinearColor=&MultiplyLinearColor(
     Linear Color=&SelectLinearColorFromArray(
      Array Sampling Mode=::NewEnumerator1
      Direct Array Index=$Particles.FlowerHue
      Color Selection Array=DataInterface() {propertyValues {ColorData [
         <rgba 1 0.4697 0.0024 1>
         <rgba 0.8333 0 0.4982 1>
         <rgba 0 0.9002 1 1>
         <rgba 0.2007 0.0659 1 1>
         <rgba 0.8177 0.2491 0 1>
         <rgba 0.0035 1 0.989 1>
         <rgba 0 1 0.3343 1>
         <rgba 1 0 0.0712 1>
         <rgba 1 0.0083 0.6752 1>
         <rgba 0 0.0252 1 1>
        ]}}
      Random Seed=?
     )
     Scale Factor=&MakeCustomVector4FromBool(Bool:NiagaraBool=$User.bIsFrontendPreview, True Vector4:Vector4f=<0.6 0.6 0.6 2.5>, False Vector4:Vector4f=<1.1 1.1 1.1 2>)
    )
    Particles.OuterColor:LinearColor=&MultiplyLinearColor(
     Linear Color=&SelectLinearColorFromArray(
      Array Sampling Mode=::NewEnumerator1
      Direct Array Index=$Particles.FlowerHue
      Color Selection Array=DataInterface() {propertyValues {ColorData [
         <rgba 0.0203 0 0.3698 1>
         <rgba 0 1 0.2287 1>
         <rgba 1 0.0001 0.0227 1>
         <rgba 0.9182 1 0 1>
         <rgba 0.0754 0.6907 0.8 1>
         <rgba 1 0.3226 0 1>
         <rgba 0.6121 0.007 1 1>
         <rgba 0 0.3588 1 1>
         <rgba 0 1 0.1269 1>
         <rgba 1 0 0.007 1>
        ]}}
      Random Seed=?
     )
     Scale Factor=&MakeCustomVector4FromBool(Bool:NiagaraBool=$User.bIsFrontendPreview, True Vector4:Vector4f=<0.6 0.6 0.6 2.5>, False Vector4:Vector4f=<1.1 1.1 1.1 2>)
    )
   ) {isSetParameters=true, module=SetVariables_#3}
   &VC_PaintedVariantsOverride(Black=<rgba 0.2517 0.2517 0.2517 1>, BurntSienna=<rgba 0.3177 0.0981 0.0392 1>)
   &InitializeParticle(
    Color Mode=::NewEnumerator0
    Lifetime Max=0.9
    Lifetime Min=0.25
    Lifetime Mode=::NewEnumerator1
    Sprite Rotation Mode=::NewEnumerator2
    Sprite Size Mode=::NewEnumerator1
    Uniform Sprite Size Max=42
    Uniform Sprite Size Min=17.5
    UsePositionOffset=false
    Write Material Random=true
   )
   &PartitionParticles(Number Of Partitions=$Emitter.PortCount)
   &SkeletalMeshLocation(
    Bone / Socket Index=$Particles.PartitionParticles.Partition
    Bone / Socket Sampling Mode=::NewEnumerator5
    MaskSpawnGroup=false
    Mesh Sampling Type=::NewEnumerator6
    Sampled Position Offset=&AddVectorToPosition(Position=<-15 0 0>, Vector=$User.DTCarOffset)
    Skeletal Mesh=$System.SkeletalMesh
   )
   &AddVelocity(
    Cone Angle=20
    Cone Axis=<0 0 1>
    Module.UseConeFalloff=false
    Module.UseVelDistribution=false
    Override Seed=false
    Rotation Mode=::NewEnumerator2
    Velocity Mode=::NewEnumerator2
    Velocity Speed=2000
    Yaw / Pitch / Roll:Vector3f=<0 82.5 0>
   )
   &ShapeLocation(MaskSpawnGroup=false, Override Seed=false, Sphere Radius=30, UseOwnerScale=true)
  ]
  particleUpdateScript [
   &ParticleState()
   &ScaleSpriteSize(
    Scale Sprite Size Mode=::NewEnumerator0
    Uniform Scale Factor=&MakeCustomFloatFromBool(
     Bool=$User.bIsFrontendPreview
     True Float=&FloatFromCurve(FloatCurve=DataInterface() {propertyValues {Curve {
         defaultValue=1
         keys [
          <at 0 0.5 RCIM_Cubic>
          {arriveTangent=-0.0087, interpMode=RCIM_Cubic, leaveTangent=-0.0087, tangentMode=RCTM_User, time=0.075, value=1}
          {arriveTangent=-0.1297, interpMode=RCIM_Cubic, leaveTangent=-0.1297, tangentMode=RCTM_User, time=0.6, value=1}
          {arriveTangent=-4.2097, interpMode=RCIM_Cubic, leaveTangent=-4.2097, tangentMode=RCTM_User, time=1, value=0}
         ]
        }}})
     False Float=&FloatFromCurve(FloatCurve=DataInterface() {propertyValues {Curve {
         defaultValue=1
         keys [
          <at 0 1 RCIM_Cubic>
          {arriveTangent=-0.123, interpMode=RCIM_Cubic, leaveTangent=-0.123, tangentMode=RCTM_User, time=0.55, value=1.25}
          {arriveTangent=-3.7842, interpMode=RCIM_Cubic, leaveTangent=-3.7842, tangentMode=RCTM_User, time=1, value=0}
         ]
        }}})
    )
   )
   &SpriteRotationRate(Rotation Rate=&RandomRangeFloat(Minimum=0.5, Maximum=360))
   &GravityForce(Gravity=&MakeVector(
     X=0
     Y=0
     Z=&FloatFromCurve(FloatCurve=DataInterface() {propertyValues {Curve {keys [
          {arriveTangent=4327.0386, interpMode=RCIM_Cubic, leaveTangent=4327.0386, tangentMode=RCTM_User, time=0, value=500}
          {arriveTangent=-162.7294, interpMode=RCIM_Cubic, leaveTangent=-162.7294, tangentMode=RCTM_User, time=0.025, value=700}
          <at 0.1 -1000 RCIM_Cubic>
         ]}}})
    )) {module=GravityForce004}
   &VortexForce(
    UseOriginOffset=false
    Vortex Axis=<-1 0 0>
    Vortex Force Amount=&Multiply_FloatByInt(
     Float=&RandomRangeFloat(Minimum=100, Maximum=500)
     Integer=&SelectIntFromArray(
      Array Sampling Mode=::NewEnumerator1
      Direct Array Index=?
      Int Selection Array=DataInterface() {propertyValues {IntData [-1,1]}}
      Random Seed=0
     )
    )
   )
   &InheritSourceMovement(
    Applied Position Scale=&Multiply_VectorByFloat(
     Vector=<1 1 1>
     Float=&FloatFromCurve(FloatCurve=DataInterface() {propertyValues {Curve {keys [{interpMode=RCIM_Cubic, tangentMode=RCTM_User, time=0.3, value=1}, {interpMode=RCIM_Cubic, tangentMode=RCTM_User, time=1, value=0}]}}})
    )
    ApplyForce=false
    ApplyPosition=true
    ApplyPositionFalloff=false
   )
   &SolveForcesAndVelocity(Clamp Velocity=false, Limit Acceleration=false)
   &SubUVAnimation(
    Direct Frame Index:NiagaraFloat=$Particles.FlowerType
    "Sprite Renderer "=DataInterface() {propertyValues {SpriteRenderer=&self:Sprites001_0.NiagaraSpriteRendererProperties_3}}
    SubUV Animation Mode:ENiagara_SubUVLookupModeV2=::NewEnumerator5
   )
   &SubUVAnimation(
    Direct Frame Index:NiagaraFloat=$Particles.FlowerType
    "Sprite Renderer "=DataInterface() {propertyValues {SpriteRenderer=&self:Sprites001_0.NiagaraSpriteRendererProperties_0}}
    SubUV Animation Mode:ENiagara_SubUVLookupModeV2=::NewEnumerator5
   ) {module=SubUVAnimation001}
   &SubUVAnimation(
    Direct Frame Index:NiagaraFloat=$Particles.FlowerType
    "Sprite Renderer "=DataInterface() {propertyValues {SpriteRenderer=&self:Sprites001_0.NiagaraSpriteRendererProperties_1}}
    SubUV Animation Mode:ENiagara_SubUVLookupModeV2=::NewEnumerator5
   ) {module=SubUVAnimation002}
   &JitterPosition(Jitter Amount=2)
   &ScaleColor(Scale RGB=&VectorFromFloat(Value=1), ScaleA=true, ScaleRGB=true)
  ]
  rendererClasses [&NiagaraSpriteRendererProperties]
  renderers [
   {rendererClass=&NiagaraSpriteRendererProperties}
   {rendererClass=&NiagaraSpriteRendererProperties, rendererIndex=1}
   {rendererClass=&NiagaraSpriteRendererProperties, rendererIndex=2}
  ]
 }
}
summary {
 systemName=NS_Boost_Flowers
 emitters [
  {emitterName=NE_VC_Boost_Ribbon_Parent, rendererClasses [&NiagaraRibbonRendererProperties]}
  {emitterName=Sprites, rendererClasses [&NiagaraSpriteRendererProperties]}
  {emitterName=Sprites001, rendererClasses [&NiagaraSpriteRendererProperties]}
 ]
 userVariables [
  User.bIsFrontendPreview {type=NiagaraBool}
  User.bIsPainted {type=NiagaraBool}
  User.ColorVariantID {type=NiagaraInt32}
  User.DTCarOffset {type=Vector3f, description="", defaultValue {struct=Vector3f, value=<0 0 0>}}
  User.PaintedColor {type=LinearColor, description="", defaultValue {struct=LinearColor, value=<rgba 1 1 1 1>}}
  User.PaintedScalar {type=NiagaraFloat, defaultValue=1, description=""}
  User.PortCount {type=NiagaraInt32, defaultValue=1}
 ]
}
